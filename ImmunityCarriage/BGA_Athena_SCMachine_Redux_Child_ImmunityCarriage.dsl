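; Child reboot machine Blueprint for ImmunityCarriage that blocks the interacting player from exiting a vehicle during a resurrection long use.
; Long-use begin applies a gameplay effect to the pawn and stores its handle.
; Interrupt, resurrection start and server end-of-use remove that effect, as does a custom event that validates the handle.
; It also overrides KnockVehicles with an empty body, disables the easy-interact button and exposes a hologram toggle event.
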(event EventBeginPlay
  (|CallParentFunction) ; This node is disabled and will not be called.\nDrag off pins to build functionality.
)

(event Collision|EventActorBeginOverlap
  (params (OtherActor :type Actor%20Object%20Reference)))

(event EventTick
  (params (DeltaSeconds :type Float%20%28single-precision%29))
  (|CallParentFunction DeltaSeconds) ; This node is disabled and will not be called.\nDrag off pins to build functionality.
)

(event Resurrection|EventOnBeginLongUse
  (params (AthenaPC :type Fort%20Player%20Controller%20Athena%20Object%20Reference))
  (|CallParentFunction 0)
  (Variables|Default|SetAthena%20PC AthenaPC)
  (bind _returnvalue (GameplayEffects|ApplyGameplayEffectToSelf :self (Ability|GetAbilitySystemComponent (Game|Player|GetPlayerPawn AthenaPC)) :GameplayEffectClass "/ImmunityCarriage/Gameplay/Effects/GE_ImmunityCarriage_BlockExitWhileRevive.GE_ImmunityCarriage_BlockExitWhileRevive_C" :Level 0.0))
  (Variables|Default|SetReviveVewhicleExitBlock _returnvalue))

(event Resurrection|EventOnInterruptLongUse
  (params (AthenaPC :type Fort%20Player%20Controller%20Athena%20Object%20Reference))
  (|CallParentFunction 0)
  (GameplayEffects|RemoveActiveGameplayEffect (Ability|GetAbilitySystemComponent (Game|Player|GetPlayerPawn AthenaPC)) (Variables|Default|GetReviveVewhicleExitBlock) -1))

(event Resurrection|EventOnResurrectionStarted
  (params (PlayerController :type Fort%20Player%20Controller%20Athena%20Object%20Reference))
  (|CallParentFunction 0)
  (GameplayEffects|RemoveActiveGameplayEffect (Ability|GetAbilitySystemComponent (Game|Player|GetPlayerPawn PlayerController)) (Variables|Default|GetReviveVewhicleExitBlock) -1))

(event Custom|ToggleHologramChild
  (CallFunction|ToggleHologram true))

(event Interaction|EventOnServerNotifyEndLongUse
  (params (InteractingPawn :type Fort%20Player%20Pawn%20Object%20Reference) (bUseCompleted :type Boolean))
  (|CallParentFunction 0 false)
  (GameplayEffects|RemoveActiveGameplayEffect (Ability|GetAbilitySystemComponent InteractingPawn) (Variables|Default|GetReviveVewhicleExitBlock) -1))

(event Custom|RemoveVehicleExitBlock
  (bind _revivevewhicleexitblock (Variables|Default|GetReviveVewhicleExitBlock))
  (bind _returnvalue (Ability|GameplayEffect|IsActiveGameplayEffectHandleValid _revivevewhicleexitblock))
  (if _returnvalue
    (GameplayEffects|RemoveActiveGameplayEffect (Ability|GetAbilitySystemComponent (Game|Player|GetPlayerPawn (Variables|Default|GetAthena%20PC))) _revivevewhicleexitblock -1)
    (else
      (Development|PrintString "No Valid Gameplay Effect"))))

(fn CanDisplayEasyInteractButton
  (return false))

(fn KnockVehicles)

(fn ConstructionScript
  (|CallParentFunction))
