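# Environment query for ImmunityCarriage that finds reboot van spawn locations in Fortnite Athena.
# It builds a coarse grid around the current safe zone indicator's center, sized by a query parameter and projected onto geometry by vertical traces.
# Tests check box overlap, two valid-surface passes (one rejecting building actors, one allowing limited height variation), safe zone containment as filter and score, and exclusion of water bodies.
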EnvQuery /ImmunityCarriage/EQ_ImmunityCarriage_RebootVanSpawnLocations
options [{
  generator {
   name=EnvQueryGenerator_SimpleGrid_0
   class=EnvQueryGenerator_SimpleGrid
   properties {
    generateAround=/Script/FortniteAIServer.FortQueryContext_AthenaCurrentSafeZoneIndicatorCenter
    spaceBetween=2048
    gridSize {
     dataBinding=&self:EnvQueryGenerator_SimpleGrid_0.AIDataProvider_QueryParams_0
     dataField=FloatValue
     defaultValue=500
    }
    projectionData {
     extentX=0
     navigationFilter=null
     postProjectionVerticalOffset=0
     projectDown=100000
     projectUp=10000
     serializedChannel=ECC_GameTraceChannel6
     traceChannel=TraceTypeQuery7
     traceMode=GeometryByChannel
    }
   }
  }
  tests [
   EnvQueryTest_Overlap_4 {
    class=EnvQueryTest_Overlap
    properties {overlapData {extentX=512, extentY=512, extentZ=512, overlapChannel=ECC_GameTraceChannel15, shapeOffset=<0 0 0>}}
   }
   FortQueryTest_ValidSurface_3 {
    class=FortQueryTest_ValidSurface
    properties {
     flatSurfaceToleranceZ=0
     radius=512
     traceOffsetDown=512
     traceOffsetUp=512
     invalidHitActorClasses [/Script/FortniteGame.BuildingActor]
     surfaceMaterials []
    }
   }
   FortQueryTest_ValidSurface_0 {
    class=FortQueryTest_ValidSurface
    properties {flatSurfaceToleranceZ=90, radius=512, traceOffsetDown=100, traceOffsetUp=100, surfaceMaterials []}
   }
   FortQueryTest_InsideAthenaSafeZone_0 {class=FortQueryTest_InsideAthenaSafeZone, properties {testPurpose=FilterAndScore}}
   EnvQueryTest_InsideWaterBody_1 {
    class=EnvQueryTest_InsideWaterBody
    properties {
     bDefineReferenceValue=false
     bIgnoreExclusionVolumes=false
     bIncludeWaves=false
     bSimpleWaves=true
     boolValue=false
     clampMaxType=None
     clampMinType=None
     filterType=Match
     floatValueMax=0
     floatValueMin=0
     multipleContextFilterOp=AllPass
     multipleContextScoreOp=AverageScore
     normalizationType=Absolute
     referenceValue=0
     scoreClampMax=0
     scoreClampMin=0
     scoringEquation=Constant
     scoringFactor=1
     testComment=""
     testPurpose=Filter
    }
   }
  ]
 }]
