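# Unlit, two-sided translucent Niagara material for the ImmunityCarriage derez trail, drawing a scrolling grid of squares that break up along the trail.
# Frac and Floor of scaled UVs create cell edges and stepped coordinates, with Perlin noise and two colors varying the squares.
# A sine-animated dissolve ramps in along V, while the Dive dynamic parameter boosts glow near V=0.
# Particle color tints and fades it, with camera fade and quality-switched depth fade.
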Material /ImmunityCarriage/FX/Materials/M_ImmunityCarriage_DerezTrail
graph {
 properties {
  bAutomaticallySetUsageInEditor=false
  bUsedWithNiagaraMeshParticles=true
  bUsedWithNiagaraRibbons=true
  bUsedWithNiagaraSprites=true
  blendMode=BLEND_Translucent
  opacityMaskClipValue=0.1
  shadingModel=MSM_Unlit
  twoSided=true
  usedShadingModels=Unlit
 }
 nodes [
  n1 = TextureCoordinate() {uTiling=0.5, vTiling=0.5}
  n2 = VectorParameter() {group=00- GridSetup, parameterName=GridUVs, defaultValue=<rgba 8 8 0 -0.2>}
  n3 = Add(
   A=@n1
   B=Multiply(A=Time(), B=MaterialFunctionCall(X=@n2.B, Y=@n2.B) {materialFunction=&MakeFloat2}.Result)
  )
  n4 = Constant2Vector() {g=2, r=10}
  n5 = OneMinus(None=TextureSample!(UVs=Multiply(A=@n3, B=@n4), Tex, Bias, Apply View MipBias) {constMipValue=1, mipValueMode=TMVM_MipBias, texture=&/T_ImmunityCarriage_Derez_Grid}.R)
  n6 = TextureCoordinate()
  LinearInterpolate(A=@n5, B=NamedRerouteUsage(), Alpha=Round(None=ComponentMask(None=@n6) {r=true}))
  MaterialFunctionCall(Emissive=NamedRerouteUsage()) {materialFunction=&MF_FX_EmissiveAdjustments}
  MaterialFunctionCall(Fade Length=Constant() {r=50}, Fade Offset=ScalarParameter() {defaultValue=100, group=01- Opacity, parameterName=CameraFadeDist}) {materialFunction=&CameraDepthFade}
  n7 = NamedRerouteUsage()
  n8 = ParticleColor()
  n9 = Constant() {r=1}
  n10 = DepthFade(FadeDistance=ScalarParameter() {defaultValue=50, group=01- Opacity, parameterName=DepthFadeDist})
  Multiply(
   A=Multiply(A=Multiply(A=@n7, B=@n8.A) {constA=1}, B=Constant() {r=0.05})
   B=MaterialFunctionCall!(
    Low _ PC Low and Med=@n9
    "High _ XB1, PS4"=@n10
    "Epic _ PC High, XSX, PS5"=@n10
    --------
    "Mobile Low _ PC Perf, Android"=@n9
    Mobile Med _ Switch=@n9
   ) {materialFunction=/Game/Global/MaterialFunctions/TechArt/MF_Fortnite_QualitySwitch_Scalar}.Result
  )
  n11 = DynamicParameter() {paramNames [Dive,Param2,Param3,Param4]}
  n12 = Divide(A=ComponentMask(None=TextureCoordinate()) {g=true}, B=ScalarParameter() {defaultValue=0.6, parameterName=DiveBoostFalloff}) {constB=0.2}
  n13 = Multiply(
   A=OneMinus(None=Saturate(None=@n12))
   B=LinearInterpolate(B=ScalarParameter() {defaultValue=30, parameterName=DiveBoost}, Alpha=@n11.Dive) {constA=5}
  )
  Multiply(A=@n11.Dive, B=@n13)
  n14 = Add(A=NamedRerouteUsage()) {constB=0}
  n15 = Saturate(None=Divide(A=Abs(None=@n14), B=ScalarParameter() {defaultValue=0.1, group=00- GridSetup, parameterName=EdgeWidth}) {constB=0.1})
  n16 = ComponentMask(None=@n15) {r=true}
  n17 = Multiply(A=ComponentMask(None=@n15) {g=true}, B=Saturate(None=Divide(A=@n16) {constB=0.5}))
  NamedRerouteDeclaration(None=Multiply(A=@n17, B=NamedRerouteUsage())) {name=Derez_Mask, nodeColor=<rgba 0.0076 0.75 1 1>}
  n18 = TextureSample!(UVs=Multiply(A=NamedRerouteUsage()) {constB=5}, Tex, Apply View MipBias) {
   samplerSource=SSM_Wrap_WorldGroupSettings
   samplerType=SAMPLERTYPE_LinearColor
   texture=/Game/Effects/SnowSandInteraction/Textures/SSI_Perlin
  }
  NamedRerouteDeclaration(None=Power(Base=@n18.RGB) {constExponent=3}) {name=GridNoise, nodeColor=<rgba 1 0.4178 0.0212 1>}
  n19 = OneMinus(None=@n17)
  n20 = ComponentMask(None=TextureCoordinate()) {g=true}
  n21 = Saturate(None=Divide(A=OneMinus(None=@n20)))
  NamedRerouteDeclaration(None=OneMinus(None=Multiply(A=@n19, B=@n21))) {name=GridEdges, nodeColor=<rgba 0 0.0625 1 1>}
  n22 = Multiply(A=MaterialFunctionCall(X=@n2.R, Y=@n2.R) {materialFunction=&MakeFloat2}.Result) {constB=2}
  n23 = Floor(None=Multiply(A=@n1, B=@n22))
  NamedRerouteDeclaration(None=Multiply(A=Divide(A=@n23, B=@n22)) {constB=2}) {name=FixedGridSteppedUVs, nodeColor=<rgba 0.6406 1 0 1>}
  n24 = VectorParameter() {group=00- GridSetup, parameterName=SquareColor_02, defaultValue=<rgba 0.003 0.052 0.15 1>}
  n25 = VectorParameter() {group=00- GridSetup, parameterName=SquareColor_01, defaultValue=<rgba 0.0121 0.3575 0.6042 1>}
  n26 = NamedRerouteUsage()
  n27 = NamedRerouteUsage()
  n28 = LinearInterpolate(
   A=VectorParameter() {group=00- GridSetup, parameterName=EdgeColor}.RGB
   B=Multiply(A=LinearInterpolate(A=@n24.RGB, B=@n25.RGB, Alpha=@n26), B=LinearInterpolate(Alpha=@n27) {constA=1, constB=5})
   Alpha=NamedRerouteUsage()
  )
  n29 = ParticleColor()
  NamedRerouteDeclaration(None=Multiply(A=Add(A=@n13), B=Multiply(A=@n28, B=@n29.RGB))) {name=Emissive, nodeColor=<rgba 0 0.3437 1 1>}
  n30 = Multiply(A=@n3, B=@n22)
  NamedRerouteDeclaration(None=Frac(None=@n30)) {name=GridUVs, nodeColor=<rgba 0 1 0.7188 1>}
  n31 = NamedRerouteUsage()
  n32 = Add(A=Saturate(None=Divide(A=@n31) {constB=0.25}), B=Time())
  n33 = Power(Base=ConstantBiasScale(None=Sine(None=@n32)), Exp=ScalarParameter() {defaultValue=1, group=01- Opacity, parameterName=DissolvePower})
  n34 = OneMinus(None=ComponentMask(None=TextureCoordinate()) {g=true})
  NamedRerouteDeclaration(None=Saturate(None=Multiply(A=@n33, B=@n34))) {name=ProgressiveDissolveGrid, nodeColor=<rgba 0 1 0.625 1>}
  n35 = Divide(A=ComponentMask(None=TextureCoordinate()) {g=true}, B=ScalarParameter() {defaultValue=0.5, group=01- Opacity, parameterName=DissolveStart}) {constB=0.5}
  NamedRerouteDeclaration(None=LinearInterpolate(B=NamedRerouteUsage(), Alpha=Saturate(None=@n35)) {constA=1}) {name=Derez_Dissolve, nodeColor=<rgba 0.4766 1 0 1>}
  n36 = Floor(None=@n30)
  NamedRerouteDeclaration(None=Multiply(A=Divide(A=@n36, B=@n22)) {constB=2}) {name=GridSteppedUVs, nodeColor=<rgba 0.2422 1 0 1>}
  Power(Base=@n5) {constExponent=5}
 ]
}
