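# Niagara system for the Immunity Carriage cancel sign, a hologram-style mesh that scales in or out.
# The active Minimal emitter burst-spawns a half-scale mesh offset upward, whose particle loops its lifetime rather than expiring.
# User.EnableHologram selects a grow or shrink curve for mesh scale, both sampled by system age rather than particle age.
# The Denied_Holo_LW emitter is disabled and has no renderers.
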NiagaraSystem /ImmunityCarriage/FX/NS_ImmunityCarriage_Cancel_Sign
dependencies {usedDataInterfaces [&NiagaraDataInterfaceCurve], usedDynamicInputs [], usedModules [&SystemState]}
emitters {
 Denied_Holo_LW {bEnabled=false, rendererClasses [], renderers []}
 Minimal {
  emitterUpdateScript [
   &EmitterState(Life Cycle Mode=::NewEnumerator1, Loop Behavior=::NewEnumerator1, Loop Duration Mode=::NewEnumerator1, UseLoopDelay=false)
   &SpawnBurst_Instantaneous(Use Spawn Probability=false, UseLoopCountLimit=false)
  ]
  particleSpawnScript [&InitializeParticle(
    Color Mode=::NewEnumerator0
    Mesh Scale Mode=::NewEnumerator3
    Mesh Uniform Scale=0.5
    Position Offset=<0 0 100>
    UsePositionOffset=true
    Write Material Random=true
   )]
  particleUpdateScript [
   &ParticleState(
    Kill Particles When Lifetime Has Elapsed=false
    Let Infinitely Lived Particles Die When Emitter Deactivates=true
    Loop Particles Lifetime=true
   )
   &ScaleMeshSize(Scale Factor=&Multiply_VectorByFloat(
     Vector=<1 1 1>
     Float=&MakeCustomFloatFromBool(
      Bool=$User.EnableHologram
      True Float=&FloatFromCurve(
       FloatCurve=DataInterface() {propertyValues {Curve {
          defaultValue=1
          keys [
           {arriveTangent=2.9251, interpMode=RCIM_Cubic, leaveTangent=2.9251, tangentMode=RCTM_User, time=0, value=0}
           {tangentMode=RCTM_User, time=1, value=1}
          ]
         }}}
       CurveIndex=$Engine.System.Age
      )
      False Float=&FloatFromCurve(
       FloatCurve=DataInterface() {propertyValues {Curve {
          defaultValue=0
          keys [<at 0 1 RCIM_Cubic>, {arriveTangent=-3.0384, interpMode=RCIM_Cubic, leaveTangent=-3.0384, tangentMode=RCTM_User, time=1, value=0}]
         }}}
       CurveIndex=$Engine.System.Age
      )
     )
    ))
  ]
  rendererClasses [&NiagaraMeshRendererProperties]
  renderers [{rendererClass=&NiagaraMeshRendererProperties, rendererIndex=0}]
 }
}
summary {
 systemName=NS_ImmunityCarriage_Cancel_Sign
 emitters [
  {bEnabled=false, emitterName=Denied_Holo_LW, rendererClasses []}
  {emitterName=Minimal, rendererClasses [&NiagaraMeshRendererProperties]}
 ]
 userVariables [User.EnableHologram {type=NiagaraBool, defaultValue=true}]
}
