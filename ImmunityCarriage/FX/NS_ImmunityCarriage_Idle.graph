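# Niagara system that draws headlight flares and glows for the Immunity Carriage's idle state.
# A burst of User.FlaresCount particles, gated on VehicleHasDriver through spawn probability, splits into two spawn groups that follow matching sockets via a socket reader.
# Four sprite renderers share each particle, with separate flare, glow and big-glow attributes derived from its color and position.
# Alpha scales with time of day, with a front-end override.
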NiagaraSystem /ImmunityCarriage/FX/NS_ImmunityCarriage_Idle
dependencies {
 usedDataInterfaces [&NiagaraDataInterfaceSocketReader]
 usedDynamicInputs []
 usedModules [
  &self:SystemSpawnScript.SystemScriptSource.SystemScriptGraph.NiagaraNodeAssignment_0.SetVariables_#1
  &SystemState
 ]
}
emitters {Headlights {
  emitterUpdateScript [
   &EmitterState(
    Inactive Response=::NewEnumerator1
    Life Cycle Mode=::NewEnumerator1
    Loop Behavior=::NewEnumerator1
    Loop Duration Mode=::NewEnumerator1
    UseLoopDelay=false
   )
   &SpawnBurst_Instantaneous(
    Spawn Count=$User.FlaresCount
    Spawn Probability=&MakeCustomFloatFromBool(Bool=$User.VehicleHasDriver, True Float=1, False Float=0)
    Use Spawn Probability=true
    UseLoopCountLimit=false
   ) {module=SpawnBurst_Instantaneous001}
  ]
  particleSpawnScript [
   &FN_SetSpawnGroup(Number of Spawn Groups=2)
   /SetVariables(Particles.MirrorAxis:Vector3f=&MaskVectorBySpawnGroup(Masked Vector=<-1 1 1>, Passthrough Vector=<1 1 1>)) {isSetParameters=true}
   &InitializeParticle(
    Color=<rgba 0.8 0.7467 0.64 0.15>
    Mesh Scale Mode=::NewEnumerator3
    Mesh Uniform Scale=10
    Sprite Size Mode=::NewEnumerator3
    Uniform Sprite Size=500
    UsePositionOffset=true
    Write Material Random=true
   )
   &CameraOffset(Camera Offset Amount=5)
   &DynamicMaterialParameters(
    Index 0 Param 1=&Multiply_Float(A=-19, B=&MakeFloatFromVector(Vector=<0 0 0>, Channel=::NewEnumerator0))
    Index 0 Param 2=-1
    Module.Param0WriteEnabled=true
    Module.Param1WriteEnabled=true
    Module.Param2WriteEnabled=false
    Module.Param3WriteEnabled=false
   )
  ]
  particleUpdateScript [
   &ParticleState(Kill Particles When Lifetime Has Elapsed=false)
   &SocketLocation(
    Orientation Sampling=::NewEnumerator1
    SamplePosOffset=true
    Sampled Position Offset=<-1 0 1>
    Socket Index=$Particles.SpawnGroup
    Socket Reader=$System.Socket Reader
    Socket Sampling Mode=::NewEnumerator3
    VelocitySpeedLimit=false
   )
   &ScaleColor(
    Scale Alpha=&FN_FrontEndFloatModifier(Gameplay=&FNTimeOfDayScalar(Night Scale=1, Morning Scale=0.7, Evening Scale=0.7), Front End=1)
    ScaleA=true
    ScaleRGB=true
   )
   /SetVariables(
    Particles.FlareAlignment:Vector3f=<0 0 1>
    Particles.FlareCameraOffset=10
    Particles.FlareColor=&MultiplyLinearColor(Linear Color=$Particles.Color, Scale Factor=&MakeVector4_FromFloat(Float=1.8))
    Particles.FlareFacing:Vector3f=<1 0 0>
    Particles.FlarePosition:NiagaraPosition=&AddVectorToPosition(Position=$Particles.Position, Vector=<1 0 -1>)
    Particles.FlareSize:Vector2f=&Multiply_Vector2D_ByFloat(Float=5, Vector2D=<100 50>)
   ) {isSetParameters=true}
   /SetVariables(
    Particles.GlowCameraOffset=5
    Particles.GlowColor=&MultiplyLinearColor(Linear Color=$Particles.Color, Scale Factor=&MakeVector4_FromFloat(Float=1))
    Particles.GlowPosition:NiagaraPosition=&AddVectorToPosition(Position=$Particles.Position, Vector=<1 0 -1>)
    Particles.GlowSize=&Vector2DFromFloat(Value=50)
    Particles.glowalignement:Vector3f=<0 0 1>
    Particles.glowfacing:Vector3f=<1 0 0>
   ) {isSetParameters=true}
   /SetVariables(
    Particles.BigGlowColor:LinearColor=&MultiplyLinearColor(
     Linear Color=$Particles.Color
     Scale Factor=&MakeVector4_FromFloat(Float=&ClampFloat(Float=&OneMinusFloat(Float=0.4)))
    )
    Particles.BigGlowSize:Vector2f=&Vector2DFromFloat(Value=150)
   ) {isSetParameters=true}
   &SpriteFacingAndAlignment(Alignment Coordinate Space=::Local, Sprite Alignment=<-1 0 0>, WriteAlignment=true, WriteFacing=false) {module=SpriteFacingAndAlignment001}
   /SetVariables(Particles.Position=&AddVectorToPosition(Position=$Particles.Position, Vector=<-50 0 0>)) {isSetParameters=true}
  ]
  rendererClasses [&NiagaraSpriteRendererProperties]
  renderers [
   {rendererClass=&NiagaraSpriteRendererProperties}
   {rendererClass=&NiagaraSpriteRendererProperties, rendererIndex=1}
   {rendererClass=&NiagaraSpriteRendererProperties, rendererIndex=2}
   {rendererClass=&NiagaraSpriteRendererProperties, rendererIndex=3}
  ]
 }}
summary {
 systemName=NS_ImmunityCarriage_Idle
 emitters [{emitterName=Headlights, rendererClasses [&NiagaraSpriteRendererProperties]}]
 userVariables [
  User.bIsFrontEnd {type=NiagaraBool}
  User.bIsFrontEndPreview {type=NiagaraBool}
  User.FlaresCount {type=NiagaraInt32, defaultValue=2}
  User.VehicleHasDriver {type=NiagaraBool, defaultValue=true}
 ]
}
